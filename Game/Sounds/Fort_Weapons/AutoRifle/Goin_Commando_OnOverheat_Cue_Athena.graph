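# Sound cue for the overheat sound of the Goin Commando auto rifle in Athena, borrowing a mounted ballistic turret overheat sample.
# A single wave player feeds the output at reduced volume, with no randomization or layering.
# It uses the close weapon projectile attenuation and the third-person ranged fire sound class.
# A concurrency override is enabled with a prevent-new rule, so extra instances are refused once the limit is reached.
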SoundCue /Game/Sounds/Fort_Weapons/AutoRifle/Goin_Commando_OnOverheat_Cue_Athena
root=WavePlayer_2
nodes [{id=WavePlayer_2, properties {soundWaveAssetPtr=/Game/Sounds/Fort_Traps/MountedTurret/Turret_Ballistic_Overheat_03}}]
properties {
 attenuationSettings=&Fort_Weapon_Projectile_Close_Athena_Att
 bOverrideConcurrency=true
 maxDistance=4950
 soundClassObject=&Fort_WeaponRanged_Fire_3P_PSM
 volumeMultiplier=0.3
 attenuationOverrides {distanceAlgorithm=NaturalSound, falloffDistance=4000, attenuationShapeExtents {x=1000}}
 concurrencyOverrides {resolutionRule=PreventNew}
}
