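# Sound cue for the looping overheat sound of the Goin Commando weapon in Athena, filed under auto rifle sounds.
# A single looping Wave Player is the root and reuses the mounted turret's overheat warning loop.
# It takes the close-range Athena weapon projectile attenuation and the critical low-ammo ranged weapon sound class.
# Its concurrency override blocks new instances at the limit rather than stopping playing ones.
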SoundCue /Game/Sounds/Fort_Weapons/AutoRifle/Goin_Commando_OverheatLoop_Cue_Athena
root=WavePlayer_0
nodes [{id=WavePlayer_0, properties {bLooping=true, soundWaveAssetPtr=/Game/Sounds/Fort_Traps/MountedTurret/MountedTurret_OverheatWarning_Loop}}]
properties {
 attenuationSettings=&Fort_Weapon_Projectile_Close_Athena_Att
 bOverrideConcurrency=true
 maxDistance=4950
 soundClassObject=/Game/Sounds/FortSoundClassesAndModes/Fort_WeaponRanged_LowAmmo_Critical_PSM
 volumeMultiplier=1
 attenuationOverrides {distanceAlgorithm=NaturalSound, falloffDistance=4000, attenuationShapeExtents {x=1000}}
 concurrencyOverrides {resolutionRule=PreventNew}
}
