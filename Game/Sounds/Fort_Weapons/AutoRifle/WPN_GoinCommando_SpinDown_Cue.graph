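# Sound cue for the spin-down sound of the GoinCommando weapon in the AutoRifle weapon audio set.
# A single wave player feeds the output directly, with no randomization or mixing.
# It plays very quietly, using the weapon reload attenuation and the third-person ranged weapon fire sound class.
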SoundCue /Game/Sounds/Fort_Weapons/AutoRifle/WPN_GoinCommando_SpinDown_Cue
root=WavePlayer_0
nodes [{id=WavePlayer_0, properties {soundWaveAssetPtr=&/WPN_GoinCommando_SpinDown}}]
properties {
 attenuationSettings=&Fort_Weapon_Reload_Att
 maxDistance=2250
 soundClassObject=&Fort_WeaponRanged_Fire_3P_PSM
 volumeMultiplier=0.05
}
