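# Sound cue for the first-person stop sound of the auto assault rifle's automatic fire.
# A single wave player plays the ar4_stop_01 sound through a modulator left at default settings.
# It uses first-person Athena projectile attenuation, the ranged first-person fire sound class, and a first-person gunfire concurrency set.
# Debug is enabled on the cue.
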SoundCue /Game/Sounds/Fort_Weapons/AutoRifle/Weapon_ASLT_Auto_Stop_1P_Cue
root=Modulator_1
nodes [{id=Modulator_1, children [WavePlayer_4]}, {id=WavePlayer_4, properties {soundWaveAssetPtr=&/ar4_stop_01}}]
properties {
 attenuationSettings=&Fort_Weapon_Projectile_FirstPerson_Athena_Att
 bDebug=true
 maxDistance=2097152
 soundClassObject=&Fort_WeaponRanged_Fire_1P_PSM
 concurrencyOverrides {maxCount=2}
 concurrencySet [&Weapon_Gunfire_FirstPerson_Conc]
}
