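# Sound cue for the looping automatic fire of the T02 Striker assault rifle, using the third-person weapon fire sound class.
# A mixer layers a near branch, which randomly picks one of two Striker fire loops under small-rifle attenuation, with a distant branch.
# The distant branch plays a generic distant rifle loop with slightly raised random pitch, its own sound class and distant projectile attenuation.
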SoundCue /Game/Sounds/Fort_Weapons/AutoRifle/Weapon_ASLT_Auto_T02_Striker_MLTI_Cue
root=Mixer_20
nodes [
 {id=Attenuation_38, children [Modulator_12], properties {attenuationSettings=/Game/Sounds/Attenuation/Fort_Weapon_Projectile_Distant_StW_Att}}
 {id=Attenuation_4, children [Random_4], properties {attenuationSettings=/Game/Sounds/Attenuation/Fort_Weapon_Rifle_Sml_Att}}
 {id=Mixer_20, children [SoundClass_2, SoundClass_41], properties {inputVolume [1,1]}}
 {id=Modulator_12, children [WavePlayer_67], properties {pitchMax=1.2, pitchMin=1.1}}
 {id=Random_4, children [WavePlayer_0,WavePlayer_1], properties {weights [1,1]}}
 {
  id=SoundClass_2
  children [Attenuation_4]
  properties {soundClassOverride=&Fort_WeaponRanged_Fire_3P_PSM}
 }
 {
  id=SoundClass_41
  children [Attenuation_38]
  properties {soundClassOverride=&Fort_WeaponRanged_Fire_Distant}
 }
 {id=WavePlayer_0, properties {bLooping=true, soundWaveAssetPtr=&/fort_striker_fireLoop_A}}
 {id=WavePlayer_1, properties {bLooping=true, soundWaveAssetPtr=&/fort_striker_fireLoop_B}}
 {id=WavePlayer_67, properties {bLooping=true, soundWaveAssetPtr=&/ar_pol_distant_loop_01}}
]
properties {bDebug=true, bOverrideConcurrency=true, maxDistance=20000, volumeMultiplier=1}
