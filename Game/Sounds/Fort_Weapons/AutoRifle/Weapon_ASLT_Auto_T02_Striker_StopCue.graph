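# Sound cue for the stop sound of the T02 Striker automatic assault rifle, likely the tail played when automatic fire ends.
# A single wave player for the Striker stop wave feeds a Modulator at the root.
# It routes through a third-person ranged weapon fire sound class with long assault rifle attenuation, overrides concurrency and has its debug flag enabled.
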SoundCue /Game/Sounds/Fort_Weapons/AutoRifle/Weapon_ASLT_Auto_T02_Striker_StopCue
root=Modulator_0
nodes [
 {id=Modulator_0, children [WavePlayer_2]}
 {id=WavePlayer_2, properties {soundWaveAssetPtr=&/fort_striker_stop}}
]
properties {
 attenuationSettings=/Game/Sounds/Attenuation/Fort_Weapon_AR_Lng_Att
 bDebug=true
 bOverrideConcurrency=true
 maxDistance=4000
 soundClassObject=&Fort_WeaponRanged_Fire_3P_PSM
 volumeMultiplier=1
}
