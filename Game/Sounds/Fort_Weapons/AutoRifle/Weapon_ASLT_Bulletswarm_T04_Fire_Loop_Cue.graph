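# Sound cue for the looping fire sound of the Bulletswarm T04 assault rifle, mixing a close layer with a distant layer.
# The close layer plays a looping stereo rifle wave through small rifle attenuation, routed to the third-person ranged fire sound class.
# The distant layer loops a separate wave, pitched slightly up by a random modulator, through distant projectile attenuation into the distant fire class.
# The cue overrides concurrency and has debug enabled.
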SoundCue /Game/Sounds/Fort_Weapons/AutoRifle/Weapon_ASLT_Bulletswarm_T04_Fire_Loop_Cue
root=Mixer_22
nodes [
 {id=Attenuation_64, children [WavePlayer_17], properties {attenuationSettings=/Game/Sounds/Attenuation/Fort_Weapon_Rifle_Sml_Att}}
 {id=Attenuation_66, children [Modulator_12], properties {attenuationSettings=/Game/Sounds/Attenuation/Fort_Weapon_Projectile_Distant_StW_Att}}
 {id=Mixer_22, children [SoundClass_42, SoundClass_43], properties {inputVolume [1,1]}}
 {id=Modulator_12, children [WavePlayer_67], properties {pitchMax=1.2, pitchMin=1.1}}
 {
  id=SoundClass_42
  children [Attenuation_64]
  properties {soundClassOverride=&Fort_WeaponRanged_Fire_3P_PSM}
 }
 {
  id=SoundClass_43
  children [Attenuation_66]
  properties {soundClassOverride=&Fort_WeaponRanged_Fire_Distant}
 }
 {id=WavePlayer_17, properties {bLooping=true, soundWaveAssetPtr=&/Weapon_AR_Loop_Stereo_02}}
 {id=WavePlayer_67, properties {bLooping=true, soundWaveAssetPtr=&/ar_pol_distant_loop_01}}
]
properties {bDebug=true, bOverrideConcurrency=true, maxDistance=20000, volumeMultiplier=1.2}
