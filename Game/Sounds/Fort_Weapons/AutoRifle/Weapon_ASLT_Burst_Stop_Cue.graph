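# Sound cue for the stop sound of a burst-fire assault rifle, routed to the third-person ranged weapon fire sound class.
# A single wave player feeds a Modulator left at default settings, adding slight pitch and volume variation to the stop sample.
# It uses long assault rifle attenuation and a concurrency rule allowing one instance per owner, stopping the oldest.
# The cue's debug flag is on.
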SoundCue /Game/Sounds/Fort_Weapons/AutoRifle/Weapon_ASLT_Burst_Stop_Cue
root=Modulator_1
nodes [{id=Modulator_1, children [WavePlayer_3]}, {id=WavePlayer_3, properties {soundWaveAssetPtr=&/ar3_stop_01}}]
properties {
 attenuationSettings=/Game/Sounds/Attenuation/Fort_Weapon_AR_Lng_Att
 bDebug=true
 maxDistance=4000
 soundClassObject=&Fort_WeaponRanged_Fire_3P_PSM
 concurrencySet [/Game/Sounds/Concurrency_Rules/OneMaximum_Owner_StopOldest]
}
