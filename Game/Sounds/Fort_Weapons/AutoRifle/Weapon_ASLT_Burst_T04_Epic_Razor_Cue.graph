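# Sound cue for the firing audio of the Tier 4 Epic Razor burst assault rifle, mixing close and distant layers.
# The close layer blends a looping fire wave with a quieter one-shot start under small-rifle attenuation and the third-person ranged fire sound class.
# The distant layer plays a looping wave through a slight upward random pitch, projectile distant attenuation and the distant fire sound class.
# The cue overrides concurrency and has debug enabled.
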SoundCue /Game/Sounds/Fort_Weapons/AutoRifle/Weapon_ASLT_Burst_T04_Epic_Razor_Cue
root=Mixer_17
nodes [
 {id=Attenuation_34, children [Modulator_12], properties {attenuationSettings=/Game/Sounds/Attenuation/Fort_Weapon_Projectile_Distant_StW_Att}}
 {id=Attenuation_49, children [Mixer_0], properties {attenuationSettings=/Game/Sounds/Attenuation/Fort_Weapon_Rifle_Sml_Att}}
 {id=Mixer_0, children [WavePlayer_1,WavePlayer_2], properties {inputVolume [1,0.6]}}
 {id=Mixer_17, children [SoundClass_32, SoundClass_35], properties {inputVolume [1,1]}}
 {id=Modulator_12, children [WavePlayer_67], properties {pitchMax=1.2, pitchMin=1.1}}
 {
  id=SoundClass_32
  children [Attenuation_49]
  properties {soundClassOverride=&Fort_WeaponRanged_Fire_3P_PSM}
 }
 {
  id=SoundClass_35
  children [Attenuation_34]
  properties {soundClassOverride=&Fort_WeaponRanged_Fire_Distant}
 }
 {id=WavePlayer_1, properties {bLooping=true, soundWaveAssetPtr=&/ar3_loop_01}}
 {id=WavePlayer_2, properties {soundWaveAssetPtr=&/ar3_start_01}}
 {id=WavePlayer_67, properties {bLooping=true, soundWaveAssetPtr=&/ar_pol_distant_loop_01}}
]
properties {bDebug=true, bOverrideConcurrency=true, maxDistance=20000, volumeMultiplier=0.8}
