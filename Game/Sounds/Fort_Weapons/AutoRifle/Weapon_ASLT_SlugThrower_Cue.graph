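# Sound cue for the Slug Thrower assault rifle's gunshot, mixing a third-person fire layer with a distant-shot layer.
# The fire layer, boosted in the mixer, randomly picks one of three slug thrower waves; the distant layer picks one of two AR Hammer distant shots.
# Each layer has its own sound class override and attenuation asset.
# A projectile-owner concurrency rule governs the cue, and its debug flag is on.
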SoundCue /Game/Sounds/Fort_Weapons/AutoRifle/Weapon_ASLT_SlugThrower_Cue
root=Mixer_3
nodes [
 {id=Attenuation_15, children [Random_5], properties {attenuationSettings=/Game/Sounds/Attenuation/Fort_Weapon_Projectile_Distant_StW_Att}}
 {id=Attenuation_9, children [Random_1], properties {attenuationSettings=/Game/Sounds/Attenuation/Fort_Weapon_Rifle_Sml_Att}}
 {id=Mixer_3, children [SoundClass_6, SoundClass_17], properties {inputVolume [1.5,1]}}
 {id=Random_1, children [WavePlayer_3,WavePlayer_4,WavePlayer_5], properties {weights [1,1,1]}}
 {id=Random_5, children [WavePlayer_22, WavePlayer_23], properties {weights [1,1]}}
 {
  id=SoundClass_17
  children [Attenuation_15]
  properties {soundClassOverride=&Fort_WeaponRanged_Fire_Distant}
 }
 {
  id=SoundClass_6
  children [Attenuation_9]
  properties {soundClassOverride=&Fort_WeaponRanged_Fire_3P_PSM}
 }
 {id=WavePlayer_22, properties {soundWaveAssetPtr=&/AR_Hammer_Distant_Shot02b}}
 {id=WavePlayer_23, properties {soundWaveAssetPtr=&/AR_Hammer_Distant_Shot02a}}
 {id=WavePlayer_3, properties {soundWaveAssetPtr=&/weapon_asslt_slugthrow_01}}
 {id=WavePlayer_4, properties {soundWaveAssetPtr=&/weapon_asslt_slugthrow_02}}
 {id=WavePlayer_5, properties {soundWaveAssetPtr=&/weapon_asslt_slugthrow_03}}
]
properties {bDebug=true, maxDistance=20000, concurrencySet [/Game/Sounds/Concurrency_Rules/Weapon_Projectile_Owner_Conc]}
