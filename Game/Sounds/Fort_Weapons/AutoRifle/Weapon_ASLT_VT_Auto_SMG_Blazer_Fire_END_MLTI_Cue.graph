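# Sound cue for the tail of automatic fire from the Vindertech Blazer SMG, apparently as heard from other players' weapons.
# A mixer layers a close end sample, using small-rifle attenuation and the third-person fire sound class, with a distant layer.
# The distant layer randomly picks one of two samples, with its own distant attenuation and sound class.
# A concurrency override caps simultaneous instances, and debug is left enabled on the cue.
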SoundCue /Game/Sounds/Fort_Weapons/AutoRifle/Weapon_ASLT_VT_Auto_SMG_Blazer_Fire_END_MLTI_Cue
root=Mixer_20
nodes [
 {id=Attenuation_29, children [Random_2], properties {attenuationSettings=/Game/Sounds/Attenuation/Fort_Weapon_Projectile_Distant_StW_Att}}
 {id=Attenuation_4, children [WavePlayer_3], properties {attenuationSettings=/Game/Sounds/Attenuation/Fort_Weapon_Rifle_Sml_Att}}
 {id=Mixer_20, children [SoundClass_2, SoundClass_30], properties {inputVolume [1,1]}}
 {id=Random_2, children [WavePlayer_4,WavePlayer_5], properties {weights [1,1]}}
 {
  id=SoundClass_2
  children [Attenuation_4]
  properties {soundClassOverride=&Fort_WeaponRanged_Fire_3P_PSM}
 }
 {
  id=SoundClass_30
  children [Attenuation_29]
  properties {soundClassOverride=&Fort_WeaponRanged_Fire_Distant}
 }
 {id=WavePlayer_3, properties {soundWaveAssetPtr=&/Vindertech_Blazer_SMG_FIRE_END_01}}
 {id=WavePlayer_4, properties {soundWaveAssetPtr=&/Vindertech_Blazer_SMG_FIRE_END_Distant_01}}
 {id=WavePlayer_5, properties {soundWaveAssetPtr=&/Vindertech_Blazer_SMG_FIRE_END_Distant_02}}
]
properties {bDebug=true, bOverrideConcurrency=true, maxDistance=20000, volumeMultiplier=0.95, concurrencyOverrides {maxCount=3}}
