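# Looping gunfire sound cue for the Goin' Commando auto rifle in Athena, with close, medium and distant layers.
# A mixer blends three looping waves, each with its own range-specific attenuation and a sound class override: 3P fire for close and medium, distant fire for the far layer.
# Medium and distant layers are pitched up and boosted, while the close layer reuses a Commando ballistic loop.
# Local and global gunfire concurrency sets limit voices.
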SoundCue /Game/Sounds/Fort_Weapons/AutoRifle/Weapon_Goin_Commando_Fire_Athena_Cue
root=Mixer_17
nodes [
 {
  id=Attenuation_0
  children [Modulator_1]
  properties {attenuationSettings=&Fort_Weapon_Projectile_Medium_Mono_Athena_Att}
 }
 {
  id=Attenuation_34
  children [Modulator_12]
  properties {attenuationSettings=&Fort_Weapon_Projectile_Distant_Mono_Athena_Att}
 }
 {id=Attenuation_49, children [Modulator_0], properties {attenuationSettings=&Fort_Weapon_Projectile_Close_Athena_Att}}
 {id=Mixer_17, children [SoundClass_32, SoundClass_0, SoundClass_35], properties {inputVolume [0.95,0.75,0.5]}}
 {id=Modulator_0, children [WavePlayer_1], properties {pitchMax=1, pitchMin=1}}
 {id=Modulator_1, children [WavePlayer_0], properties {pitchMax=1.3, pitchMin=1.3, volumeMax=2.05, volumeMin=1.95}}
 {id=Modulator_12, children [WavePlayer_3], properties {pitchMax=1.3, pitchMin=1.3, volumeMax=2.05, volumeMin=1.95}}
 {
  id=SoundClass_0
  children [Attenuation_0]
  properties {soundClassOverride=&Fort_WeaponRanged_Fire_3P_PSM}
 }
 {
  id=SoundClass_32
  children [Attenuation_49]
  properties {soundClassOverride=&Fort_WeaponRanged_Fire_3P_PSM}
 }
 {
  id=SoundClass_35
  children [Attenuation_34]
  properties {soundClassOverride=&Fort_WeaponRanged_Fire_Distant}
 }
 {
  id=WavePlayer_0
  properties {bLooping=true, soundWaveAssetPtr=&/WPN_Goin_Commando_Fire_Loop_03_mediumNew_athena}
 }
 {id=WavePlayer_1, properties {bLooping=true, soundWaveAssetPtr=/Game/Sounds/Fort_Weapons/AutoRifle/Commando/Commando_Ballistic_FireLoop_01}}
 {
  id=WavePlayer_3
  properties {bLooping=true, soundWaveAssetPtr=&/WPN_Goin_Commando_Fire_Loop_03_distantNew_athena}
 }
]
properties {
 bDebug=true
 maxDistance=27500
 soundClassObject=&Fort_WeaponRanged
 volumeMultiplier=1
 concurrencySet [&Weapon_Gunfire_Local_Conc, &Weapon_Gunfire_Global_Conc]
}
