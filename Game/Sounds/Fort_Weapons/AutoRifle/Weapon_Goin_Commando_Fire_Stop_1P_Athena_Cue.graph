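# Sound cue for the first-person sound that ends the Goin Commando auto rifle's looping fire in Athena.
# It plays a single close-perspective loop-end wave through a Modulator left at its default settings.
# The cue uses first-person projectile attenuation, a first-person ranged fire sound class and the shared first-person gunfire concurrency set.
# The debug flag is left switched on.
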SoundCue /Game/Sounds/Fort_Weapons/AutoRifle/Weapon_Goin_Commando_Fire_Stop_1P_Athena_Cue
root=Modulator_0
nodes [
 {id=Modulator_0, children [WavePlayer_10]}
 {id=WavePlayer_10, properties {soundWaveAssetPtr=&/WPN_Goin_Commando_Fire_Loop_03_end_close_athena}}
]
properties {
 attenuationSettings=&Fort_Weapon_Projectile_FirstPerson_Athena_Att
 bDebug=true
 maxDistance=2097152
 soundClassObject=&Fort_WeaponRanged_Fire_1P_PSM
 volumeMultiplier=1
 concurrencyOverrides {bLimitToOwner=true, maxCount=3}
 concurrencySet [&Weapon_Gunfire_FirstPerson_Conc]
}
