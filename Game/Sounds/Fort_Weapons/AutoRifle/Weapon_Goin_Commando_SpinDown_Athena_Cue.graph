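# Sound cue for the Goin Commando auto rifle's spin-down sound in Athena, using third-person ranged weapon sound class overrides.
# A mixer layers a close wave and a distant wave, each with its own modulator and attenuation (close projectile versus medium-range minigun spin).
# The distant branch uses a quality level switch to play an optimized wave on some quality tiers.
# The cue overrides concurrency and has its debug flag enabled.
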SoundCue /Game/Sounds/Fort_Weapons/AutoRifle/Weapon_Goin_Commando_SpinDown_Athena_Cue
root=Mixer_17
nodes [
 {
  id=Attenuation_0
  children [Modulator_1]
  properties {attenuationSettings=/Game/Sounds/Attenuation/Fort_Weapon_Projectile_Medium_MinigunSpin_Athena_Att}
 }
 {id=Attenuation_49, children [Modulator_0], properties {attenuationSettings=&Fort_Weapon_Projectile_Close_Athena_Att}}
 {id=Mixer_17, children [SoundClass_32, SoundClass_0], properties {inputVolume [1,1]}}
 {id=Modulator_0, children [WavePlayer_1]}
 {id=Modulator_1, children [QualityLevel_0]}
 {id=QualityLevel_0, children [WavePlayer_2,WavePlayer_2,WavePlayer_0,WavePlayer_0]}
 {
  id=SoundClass_0
  children [Attenuation_0]
  properties {soundClassOverride=&Fort_WeaponRanged_Fire_3P_PSM}
 }
 {
  id=SoundClass_32
  children [Attenuation_49]
  properties {soundClassOverride=&Fort_WeaponRanged_Fire_3P_PSM}
 }
 {id=WavePlayer_0, properties {soundWaveAssetPtr=&/WPN_GoinCommando_SpinDown_distant_athena_opt}}
 {id=WavePlayer_1, properties {soundWaveAssetPtr=&/WPN_GoinCommando_SpinDown_close_athena}}
 {id=WavePlayer_2, properties {soundWaveAssetPtr=&/WPN_GoinCommando_SpinDown_distant_athena}}
]
properties {
 bDebug=true
 bOverrideConcurrency=true
 maxDistance=4950
 soundClassObject=&Fort_WeaponRanged
 volumeMultiplier=0.3
}
