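# Sound cue for the magazine insert step of a sniper rifle reload in Fortnite's weapon foley set.
# A Modulator wraps a single wave player for the clip-in sample, leaving its pitch and volume variation at default ranges.
# Playback uses the shared weapon reload attenuation and WeaponFoley sound class at reduced volume, with concurrency overridden and debug enabled.
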SoundCue /Game/Sounds/Fort_Weapons/AutoRifle/Weapon_Sniper_Reload_Insert_Cue
root=Modulator_0
nodes [
 {id=Modulator_0, children [WavePlayer_0]}
 {id=WavePlayer_0, properties {soundWaveAssetPtr=/Game/Sounds/Fort_Weapons/AutoRifle/Reload_Elements/rifle_sniper_single_clip_in}}
]
properties {
 attenuationSettings=&Fort_Weapon_Reload_Att
 bDebug=true
 bOverrideConcurrency=true
 maxDistance=2250
 soundClassObject=&Fort_WeaponFoley
 volumeMultiplier=0.401
}
