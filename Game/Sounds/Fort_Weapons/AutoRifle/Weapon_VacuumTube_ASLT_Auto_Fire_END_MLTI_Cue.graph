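# Sound cue for the end of automatic fire on the Vacuum Tube assault rifle, using third-person and distant weapon-fire sound classes.
# A mixer plays two layers at once: one of three random fire shots with small-rifle attenuation, and one of two random distant tails with projectile-distant attenuation.
# Overridden concurrency allows two instances per owner, stopping the oldest, and the debug flag is left on.
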SoundCue /Game/Sounds/Fort_Weapons/AutoRifle/Weapon_VacuumTube_ASLT_Auto_Fire_END_MLTI_Cue
root=Mixer_0
nodes [
 {id=Attenuation_0, children [Random_1], properties {attenuationSettings=/Game/Sounds/Attenuation/Fort_Weapon_Rifle_Sml_Att}}
 {id=Attenuation_29, children [Random_2], properties {attenuationSettings=/Game/Sounds/Attenuation/Fort_Weapon_Projectile_Distant_StW_Att}}
 {id=Mixer_0, children [SoundClass_5, SoundClass_30], properties {inputVolume [1,1]}}
 {id=Random_1, children [WavePlayer_9, WavePlayer_10, WavePlayer_11], properties {weights [1,1,1]}}
 {id=Random_2, children [WavePlayer_1,WavePlayer_2], properties {weights [1,1]}}
 {
  id=SoundClass_30
  children [Attenuation_29]
  properties {soundClassOverride=&Fort_WeaponRanged_Fire_Distant}
 }
 {
  id=SoundClass_5
  children [Attenuation_0]
  properties {soundClassOverride=&Fort_WeaponRanged_Fire_3P_PSM}
 }
 {id=WavePlayer_1, properties {soundWaveAssetPtr=&/Weapon_AR_VT_FIRE_END_Distant_01}}
 {id=WavePlayer_10, properties {soundWaveAssetPtr=&/Weapon_VT_AR_FIRE_03}}
 {id=WavePlayer_11, properties {soundWaveAssetPtr=&/Weapon_VT_AR_FIRE_02}}
 {id=WavePlayer_2, properties {soundWaveAssetPtr=&/Weapon_AR_VT_FIRE_END_Distant_02}}
 {id=WavePlayer_9, properties {soundWaveAssetPtr=&/Weapon_VT_AR_FIRE_01}}
]
properties {
 bDebug=true
 bOverrideConcurrency=true
 maxDistance=20000
 volumeMultiplier=1.3
 concurrencyOverrides {bLimitToOwner=true, maxCount=2, resolutionRule=StopOldest}
}
